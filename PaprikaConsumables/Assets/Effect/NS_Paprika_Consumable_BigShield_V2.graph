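# Niagara effect for the Big Shield consumable: a blue liquid pour, ramped up over system age and using a sampled TargetPawn skeletal mesh.
# A GPU PIC/FLIP fluid simulation runs on camera-facing 2D grids with render targets for surfacing.
# BakeFluid sprite emitters, apparently non-Epic quality fallbacks, shift particles by the sampled mesh position's per-frame movement.
# GPU (colliding) and CPU drip emitters add speed-stretched droplets; one BakeFluid emitter is disabled.
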NiagaraSystem /PaprikaConsumables/Assets/Effect/NS_Paprika_Consumable_BigShield_V2
dependencies {
 usedDataInterfaces [
  &NiagaraDataInterfaceCurve
  &NiagaraDataInterfaceVector2DCurve
  /Script/Niagara.NiagaraDataInterfaceGrid2DCollection
  /Script/Niagara.NiagaraDataInterfaceRenderTarget2D
  &NiagaraDataInterfaceCamera
  /Script/Niagara.NiagaraDataInterfaceRasterizationGrid3D
  &NiagaraDataInterfacePlatformSet
  /Script/Niagara.NiagaraDataInterfaceEmitterProperties
 ]
 usedDynamicInputs []
 usedModules [
  &SystemState
  &self:SystemSpawnScript.SystemScriptSource.SystemScriptGraph.NiagaraNodeAssignment_0.SetVariables_#1
  &self:SystemSpawnScript.SystemScriptSource.SystemScriptGraph.NiagaraNodeAssignment_1.SetVariables_#2
 ]
}
emitters {
 BakeFluid_NoneEpic {
  emitterUpdateScript [
   &EmitterState()
   &SpawnRate(
    SpawnRate=&FloatFromCurve(
     FloatCurve=DataInterface() {propertyValues {Curve {keys [{arriveTangent=7.3353, interpMode=RCIM_Cubic, leaveTangent=7.3353, tangentMode=RCTM_User, time=0, value=0}, <at 0.5 1 RCIM_Cubic>]}}}
     CurveIndex=&Subtract_Float(A=$Engine.System.Age, B=1)
     Scale Curve=25
    )
    Use Spawn Probability=false
   )
   &SampleSkeletalMesh(
    AdditionalYawPitchRoll=false
    Bone Sampling Mode:ENiagaraBoneSamplingMode=::NewEnumerator4
    Skeletal Mesh:NiagaraDataInterfaceSkeletalMesh=$User.TargetPawn
   )
  ]
  particleSpawnScript [
   &InitializeParticle(
    Lifetime Max=1
    Lifetime Min=0.8
    Lifetime Mode=::NewEnumerator1
    Sprite Rotation Mode=::NewEnumerator2
    Sprite Size Mode=::NewEnumerator1
    Sprite UV Mode=::NewEnumerator3
    Uniform Sprite Size Max=70
    Uniform Sprite Size Min=40
    UsePositionOffset=false
    Write Material Random=true
   )
   &CurlNoiseForce(
    Noise Frequency=4
    Noise Strength=&FloatFromCurve(
     FloatCurve=DataInterface() {propertyValues {Curve {keys [{interpMode=RCIM_Cubic, tangentMode=RCTM_Break, time=0, value=0}, <at 0.9 1>]}}}
     CurveIndex=&Multiply_Float(A=$System.Age, B=1)
     Scale Curve=50
    )
    Pan Noise=true
    Pan Noise Field=<0 0 5>
   )
   &AddVelocity(
    Velocity=&MakeVector(
     X=50
     Y=&Multiply_Float(
      A=150
      B=&Lerp_Float(
       A=-1
       B=1
       Alpha=&FloatFromCurve(
        FloatCurve=DataInterface() {propertyValues {Curve {keys [<at 0 0 RCIM_Cubic>,<at 0.4506 0>,<at 0.539 1>,<at 1 1 RCIM_Cubic>]}}}
        CurveIndex=&RandomRangeFloat(Minimum=0, Maximum=1)
       )
      )
     )
     Z=50
    )
    Velocity Speed Scale=$System.VelocityMultiplier
   )
   &CameraOffset()
   &ApplyInitialForces()
   /SetVariables(Particles.LastEmitterPosition:NiagaraPosition=$Emitter.SampleSkeletalMesh.SampledPosition) {isSetParameters=true, module=SetVariables_#3}
  ]
  particleUpdateScript [
   &ParticleState()
   &ScaleSpriteSize(
    Uniform Curve Scale=4
    Uniform Curve Sprite Scale=DataInterface() {propertyValues {Curve {keys [
        {arriveTangent=1.5276, interpMode=RCIM_Cubic, leaveTangent=1.5276, tangentMode=RCTM_User, time=0, value=0}
        {arriveTangent=1.7514, interpMode=RCIM_Cubic, leaveTangent=1.7514, tangentMode=RCTM_User, time=1, value=1}
       ]}}}
   )
   &DynamicMaterialParameters(
    Index 0 Param 1=&FloatFromCurve(FloatCurve=DataInterface() {propertyValues {Curve {keys [{tangentMode=RCTM_User, time=0, value=0.1}, {tangentMode=RCTM_User, time=1, value=1}]}}})
    Index 0 Param 2=&FloatFromCurve(FloatCurve=DataInterface() {propertyValues {Curve {keys [
         {arriveTangent=1.2796, interpMode=RCIM_Cubic, leaveTangent=1.2796, tangentMode=RCTM_User, time=0, value=0}
         {arriveTangent=0.0224, interpMode=RCIM_Cubic, leaveTangent=0.0224, tangentMode=RCTM_User, time=1, value=1}
        ]}}})
    Index 0 Param 3=&FloatFromCurve(FloatCurve=DataInterface() {propertyValues {Curve {keys [
         {arriveTangent=0.4361, interpMode=RCIM_Cubic, tangentMode=RCTM_Break, time=0.5, value=0}
         {arriveTangent=-1.6488, interpMode=RCIM_Cubic, tangentMode=RCTM_Break, time=1, value=-0.5}
        ]}}})
    Module.Param0WriteEnabled=true
    Module.Param1WriteEnabled=true
    Module.Param2WriteEnabled=true
    Module.Param3WriteEnabled=true
   )
   &GravityForce(Gravity=<0 0 -650>)
   &Drag(Use Linear Drag=true, Use Rotational Drag=true)
   &SolveForcesAndVelocity(Clamp Velocity=false, Limit Acceleration=false)
   /SetVariables(Particles.Position=&AddVectorToPosition(
     Position=$Particles.Position
     Vector=&SubtractVectorFromPosition(
      Position=$Emitter.SampleSkeletalMesh.SampledPosition
      Vector=Linked() {linkedVariable {name=Particles.LastEmitterPosition, type=NiagaraPosition}}
     )
    )) {isSetParameters=true, module=SetVariables_#4}
   /SetVariables(Particles.LastEmitterPosition:NiagaraPosition=$Emitter.SampleSkeletalMesh.SampledPosition) {isSetParameters=true, module=SetVariables_#5}
   &ScaleColor(
    Scale Alpha=&FloatFromCurve(
     FloatCurve=DataInterface() {propertyValues {Curve {keys [
         {interpMode=RCIM_Cubic, tangentMode=RCTM_User, time=0, value=0}
         <at 0.2 1 RCIM_Cubic>
         {arriveTangent=-4.106, interpMode=RCIM_Cubic, leaveTangent=-4.106, tangentMode=RCTM_User, time=1, value=0}
        ]}}}
     Scale Curve=&Multiply_Float(A=1, B=0.8)
    )
    ScaleA=true
    ScaleRGB=true
   )
  ]
  rendererClasses [&NiagaraSpriteRendererProperties]
  renderers [{rendererClass=&NiagaraSpriteRendererProperties}]
 }
 BakeFluid_PoringStar_NoneEpic {
  emitterUpdateScript [
   &EmitterState()
   &SpawnRate(
    SpawnRate=&FloatFromCurve(
     FloatCurve=DataInterface() {propertyValues {Curve {keys [{arriveTangent=3.4706, interpMode=RCIM_Cubic, leaveTangent=3.4706, tangentMode=RCTM_User, time=0, value=0}, <at 0.5 1 RCIM_Cubic>]}}}
     CurveIndex=$System.Age
     Scale Curve=12
    )
    Use Spawn Probability=false
   )
   &SampleSkeletalMesh(
    AdditionalYawPitchRoll=false
    Bone Sampling Mode:ENiagaraBoneSamplingMode=::NewEnumerator4
    Skeletal Mesh:NiagaraDataInterfaceSkeletalMesh=$User.TargetPawn
   )
  ]
  particleSpawnScript [
   &InitializeParticle(
    Lifetime Max=0.7
    Lifetime Min=0.6
    Lifetime Mode=::NewEnumerator1
    Sprite Rotation Mode=::NewEnumerator2
    Sprite Size Mode=::NewEnumerator1
    Sprite UV Mode=::NewEnumerator3
    Uniform Sprite Size Max=70
    Uniform Sprite Size Min=40
    UsePositionOffset=false
    Write Material Random=true
   )
   &CurlNoiseForce(
    Noise Frequency=4
    Noise Strength=&FloatFromCurve(
     FloatCurve=DataInterface() {propertyValues {Curve {keys [{interpMode=RCIM_Cubic, tangentMode=RCTM_Break, time=0, value=0}, <at 0.9 1>]}}}
     CurveIndex=&Multiply_Float(A=$System.Age, B=1)
     Scale Curve=50
    )
    Pan Noise=true
    Pan Noise Field=<0 0 5>
   )
   &AddVelocity(Velocity=<30 0 50>)
   &CameraOffset(Camera Offset Amount=20)
   &ApplyInitialForces()
   /SetVariables(Particles.LastEmitterPosition:NiagaraPosition=$Emitter.SampleSkeletalMesh.SampledPosition) {isSetParameters=true, module=SetVariables_#3}
  ]
  particleUpdateScript [
   &ParticleState()
   &ScaleSpriteSize(Uniform Curve Sprite Scale=DataInterface() {propertyValues {Curve {keys [
        {arriveTangent=0.767, interpMode=RCIM_Cubic, leaveTangent=0.767, tangentMode=RCTM_User, time=0, value=0.45}
        {interpMode=RCIM_Cubic, tangentMode=RCTM_User, time=1, value=1}
       ]}}})
   &DynamicMaterialParameters(
    Index 0 Param 1=&FloatFromCurve(FloatCurve=DataInterface() {propertyValues {Curve {keys [<at 0 0.01>, {arriveTangent=0.6025, interpMode=RCIM_Cubic, leaveTangent=-0.6448, tangentMode=RCTM_Break, time=1, value=1}]}}})
    Index 0 Param 2=&FloatFromCurve(FloatCurve=DataInterface() {propertyValues {Curve {keys [{arriveTangent=0.4361, time=0.2, value=0.05}, <at 1 0.7>]}}})
    Index 0 Param 3=&FloatFromCurve(FloatCurve=DataInterface() {propertyValues {Curve {keys [
         {arriveTangent=0.4361, interpMode=RCIM_Cubic, tangentMode=RCTM_Break, time=0.5, value=0}
         {arriveTangent=-1.6488, interpMode=RCIM_Cubic, tangentMode=RCTM_Break, time=1, value=-0.5}
        ]}}})
    Module.Param0WriteEnabled=true
    Module.Param1WriteEnabled=true
    Module.Param2WriteEnabled=true
    Module.Param3WriteEnabled=true
   )
   &GravityForce(Gravity=<0 0 -200>)
   &SolveForcesAndVelocity(Clamp Velocity=false, Limit Acceleration=false)
   /SetVariables(Particles.Position=&AddVectorToPosition(
     Position=$Particles.Position
     Vector=&SubtractVectorFromPosition(
      Position=$Emitter.SampleSkeletalMesh.SampledPosition
      Vector=Linked() {linkedVariable {name=Particles.LastEmitterPosition, type=NiagaraPosition}}
     )
    )) {isSetParameters=true, module=SetVariables_#4}
   /SetVariables(Particles.LastEmitterPosition:NiagaraPosition=$Emitter.SampleSkeletalMesh.SampledPosition) {isSetParameters=true, module=SetVariables_#5}
  ]
  rendererClasses [&NiagaraSpriteRendererProperties]
  renderers [{rendererClass=&NiagaraSpriteRendererProperties}]
 }
 BakeFluid_PoringStart {
  bEnabled=false
  emitterUpdateScript [
   &EmitterState()
   &SpawnRate(
    SpawnRate=&FloatFromCurve(
     FloatCurve=DataInterface() {propertyValues {Curve {keys [{arriveTangent=3.4706, interpMode=RCIM_Cubic, leaveTangent=3.4706, tangentMode=RCTM_User, time=0, value=0}, <at 0.5 1 RCIM_Cubic>]}}}
     CurveIndex=$System.Age
     Scale Curve=12
    )
    Use Spawn Probability=false
   )
   &SampleSkeletalMesh(
    AdditionalYawPitchRoll=false
    Bone Sampling Mode:ENiagaraBoneSamplingMode=::NewEnumerator4
    Skeletal Mesh:NiagaraDataInterfaceSkeletalMesh=$User.TargetPawn
   )
  ]
  particleSpawnScript [
   &InitializeParticle(
    Lifetime Max=0.7
    Lifetime Min=0.6
    Lifetime Mode=::NewEnumerator1
    Sprite Rotation Mode=::NewEnumerator2
    Sprite Size Mode=::NewEnumerator1
    Sprite UV Mode=::NewEnumerator3
    Uniform Sprite Size Max=70
    Uniform Sprite Size Min=40
    UsePositionOffset=false
    Write Material Random=true
   )
   &CurlNoiseForce(
    Noise Frequency=4
    Noise Strength=&FloatFromCurve(
     FloatCurve=DataInterface() {propertyValues {Curve {keys [{interpMode=RCIM_Cubic, tangentMode=RCTM_Break, time=0, value=0}, <at 0.9 1>]}}}
     CurveIndex=&Multiply_Float(A=$System.Age, B=1)
     Scale Curve=50
    )
    Pan Noise=true
    Pan Noise Field=<0 0 5>
   )
   &AddVelocity(Velocity=<30 0 50>)
   &CameraOffset(Camera Offset Amount=20)
   &ApplyInitialForces()
   /SetVariables(Particles.LastEmitterPosition:NiagaraPosition=$Emitter.SampleSkeletalMesh.SampledPosition) {isSetParameters=true, module=SetVariables_#3}
  ]
  particleUpdateScript [
   &ParticleState()
   &ScaleSpriteSize(Uniform Curve Sprite Scale=DataInterface() {propertyValues {Curve {keys [
        {arriveTangent=0.767, interpMode=RCIM_Cubic, leaveTangent=0.767, tangentMode=RCTM_User, time=0, value=0.45}
        {interpMode=RCIM_Cubic, tangentMode=RCTM_User, time=1, value=1}
       ]}}})
   &DynamicMaterialParameters(
    Index 0 Param 1=&FloatFromCurve(FloatCurve=DataInterface() {propertyValues {Curve {keys [<at 0 0.01>, {arriveTangent=0.6025, interpMode=RCIM_Cubic, leaveTangent=-0.6448, tangentMode=RCTM_Break, time=1, value=1}]}}})
    Index 0 Param 2=&FloatFromCurve(FloatCurve=DataInterface() {propertyValues {Curve {keys [{arriveTangent=0.4361, time=0.2, value=0.05}, <at 1 0.7>]}}})
    Index 0 Param 3=&FloatFromCurve(FloatCurve=DataInterface() {propertyValues {Curve {keys [
         {arriveTangent=0.4361, interpMode=RCIM_Cubic, tangentMode=RCTM_Break, time=0.5, value=0}
         {arriveTangent=-1.6488, interpMode=RCIM_Cubic, tangentMode=RCTM_Break, time=1, value=-0.5}
        ]}}})
    Module.Param0WriteEnabled=true
    Module.Param1WriteEnabled=true
    Module.Param2WriteEnabled=true
    Module.Param3WriteEnabled=true
   )
   &GravityForce(Gravity=<0 0 -60>)
   &SolveForcesAndVelocity(Clamp Velocity=false, Limit Acceleration=false)
   /SetVariables(Particles.Position=&AddVectorToPosition(
     Position=$Particles.Position
     Vector=&SubtractVectorFromPosition(
      Position=$Emitter.SampleSkeletalMesh.SampledPosition
      Vector=Linked() {linkedVariable {name=Particles.LastEmitterPosition, type=NiagaraPosition}}
     )
    )) {isSetParameters=true, module=SetVariables_#4}
   /SetVariables(Particles.LastEmitterPosition:NiagaraPosition=$Emitter.SampleSkeletalMesh.SampledPosition) {isSetParameters=true, module=SetVariables_#5}
  ]
  rendererClasses [&NiagaraSpriteRendererProperties]
  renderers [{rendererClass=&NiagaraSpriteRendererProperties}]
 }
 Drips_CPU_Low {
  emitterUpdateScript [
   &EmitterState()
   &SpawnRate(SpawnRate=80, Use Spawn Probability=false)
  ]
  particleSpawnScript [
   /Niagara/Modules/Spawn/Initialization/InitializeParticle(
    Lifetime=&RandomRangeFloat(Minimum=0.8, Maximum=1.5, Evaluation Type=::NewEnumerator1)
    Module.Write Color=true
    Module.Write Lifetime=true
    Module.Write Mass=true
    Module.Write Material Random=true
    Module.Write Position=false
    Module.Write Scale=false
    Module.Write SpriteRotation=false
    Module.Write SpriteSize=true
    Sprite Size=&RandomRangeVector2D(Minimum=<0.85 1>, Maximum=<2 3>, Evaluation Type=::NewEnumerator1)
   )
   &AddVelocity(
    Scale Added Velocity=&VectorFromFloat(Value=$System.VelocityMultiplier)
    Velocity=&RandomRangeVector(Minimum=<25 -10 -10>, Maximum=<50 10 10>, Evaluation Type=::NewEnumerator1)
   )
   /Niagara/Modules/Spawn/Location/SphereLocation(MaskSpawnGroup=false, Module.Override Randomness=false, Sphere Radius=2, Surface Only=false)
   &AddVelocity(Velocity=&Multiply_VectorByFloat(
     Vector=$Output.SphereLocation.SphereVector
     Float=&RandomRangeFloat(Minimum=10, Maximum=20, Evaluation Type=::NewEnumerator1)
    )) {module=AddVelocity001}
  ]
  particleUpdateScript [
   &ParticleState()
   &AccelerationForce(Acceleration=<0 0 -980>)
   &Drag(Use Linear Drag=true, Use Rotational Drag=false)
   &SolveForcesAndVelocity(Clamp Velocity=false, Limit Acceleration=false)
   &Color(Color=&MakeLinearColorFromVectorAndFloat(
     "Vector (RGB)"=&VectorFromFloat(Value=&FNTimeOfDayScalar(Night Scale=1.5, Morning Scale=3, Evening Scale=3))
     "Float (Alpha)"=&ClampFloat(Float=&FloatFromCurve(FloatCurve=DataInterface() {propertyValues {Curve {keys [{arriveTangent=-1, time=0, value=1}, <at 1 0>]}}}))
    ))
   &ScaleSpriteSizeBySpeed(Max Scale Factor=<1 25>, Source Velocity:Vector3f=&Multiply_Vector(A=<0 0.05 1>, B=$Particles.Velocity))
   &ScaleSpriteSize(Scale Factor=&Vector2DFromCurve(Vector2Curve=DataInterface() {propertyValues {XCurve {keys [<at 0.9 1>,<at 1 0>]}, YCurve {keys [<at 0.9 1>,<at 1 0>]}}}, ScaleCurve=<1 1>))
  ]
  rendererClasses [&NiagaraSpriteRendererProperties]
  renderers [{rendererClass=&NiagaraSpriteRendererProperties}]
 }
 Drips_GPU {
  simTarget=GPUComputeSim
  emitterUpdateScript [
   &EmitterState()
   &SpawnRate(SpawnRate=140, Use Spawn Probability=false)
  ]
  particleSpawnScript [
   &InitializeParticle(
    Lifetime=&RandomRangeFloat(Minimum=0.5, Maximum=1, Evaluation Type=::NewEnumerator1)
    Module.Write Color=true
    Module.Write Lifetime=true
    Module.Write Mass=true
    Module.Write Material Random=true
    Module.Write Position=false
    Module.Write Scale=false
    Module.Write SpriteRotation=false
    Module.Write SpriteSize=true
    Sprite Size=&RandomRangeVector2D(Minimum=<1 1>, Maximum=<3 3>, Evaluation Type=::NewEnumerator1)
   )
   &AddVelocity(
    Scale Added Velocity=&VectorFromFloat(Value=$System.VelocityMultiplier)
    Velocity=&RandomRangeVector(Minimum=<25 -10 -10>, Maximum=<50 10 10>, Evaluation Type=::NewEnumerator1)
   )
   &SphereLocation(MaskSpawnGroup=false, Module.Override Randomness=false, Sphere Radius=2, Surface Only=false)
   &AddVelocity(Velocity=&Multiply_VectorByFloat(
     Vector=$Output.SphereLocation.SphereVector
     Float=&RandomRangeFloat(Minimum=10, Maximum=20, Evaluation Type=::NewEnumerator1)
    )) {module=AddVelocity001}
   &InheritVelocity(Speed Limit=true, UseSpeedThreshold=false) {enabled=false}
  ]
  particleUpdateScript [
   &ParticleState()
   &InheritSourceMovement(Applied Position Scale=<0.7 0.7 0.7>, ApplyForce=false, ApplyPosition=true, ApplyPositionFalloff=false)
   &AccelerationForce(Acceleration=<0 0 -980>)
   &Drag(Use Linear Drag=true, Use Rotational Drag=false)
   &Collision(
    Age Colliding Particles=false
    Kill Offscreen Particles=true
    KillOnCollision=false
    Module.Control Rest State Via Surface Normal=false
    Module.EnableMaxCollisionCount=false
    Module.Randomize Collision Normal Vector=false
    Restitution=0
   )
   &SolveForcesAndVelocity(Clamp Velocity=false, Limit Acceleration=false)
   &Color(Color=&MakeLinearColorFromVectorAndFloat(
     "Vector (RGB)"=&VectorFromFloat(Value=&FNTimeOfDayScalar(Night Scale=1.5, Morning Scale=3, Evening Scale=3))
     "Float (Alpha)"=&ClampFloat(Float=&FloatFromCurve(FloatCurve=DataInterface() {propertyValues {Curve {keys [{arriveTangent=-1, time=0, value=1}, <at 1 0>]}}}))
    ))
   &ScaleSpriteSizeBySpeed(Max Scale Factor=<1 25>, Source Velocity:Vector3f=&Multiply_Vector(A=<0 0.05 1>, B=$Particles.Velocity))
   &ScaleSpriteSize(Scale Factor=&Vector2DFromCurve(Vector2Curve=DataInterface() {propertyValues {XCurve {keys [<at 0.9 1>,<at 1 0>]}, YCurve {keys [<at 0.9 1>,<at 1 0>]}}}, ScaleCurve=<1 1>))
  ]
  rendererClasses [&NiagaraSpriteRendererProperties]
  renderers [{rendererClass=&NiagaraSpriteRendererProperties}]
 }
 FN_PIC_FLIP_Fluid_Stream002 {
  simTarget=GPUComputeSim
  emitterSpawnScript [
   /SetVariables(
    Emitter.PressureGrid=DataInterface() {propertyValues {
      ClearBeforeNonIterationStage=true
      NumAttributes=0
      NumCellsMaxAxis=3
      NumCellsX=3
      NumCellsY=3
      OverrideBufferFormat=Float
      PreviewAttribute=None
      SetGridFromMaxAxis=false
      bOverrideFormat=true
      bPreviewGrid=false
      RenderTargetUserParameter {parameter {name=None, typeDefHandle {registeredTypeIndex=4}, varData []}}
      WorldBBoxSize=<100 100>
     }}
    Emitter.VelocityGrid=DataInterface() {propertyValues {
      ClearBeforeNonIterationStage=true
      NumAttributes=0
      NumCellsMaxAxis=3
      NumCellsX=128
      NumCellsY=64
      OverrideBufferFormat=Float
      PreviewAttribute=None
      SetGridFromMaxAxis=false
      bOverrideFormat=false
      bPreviewGrid=false
      RenderTargetUserParameter {parameter {name=None, typeDefHandle {registeredTypeIndex=4}, varData []}}
      WorldBBoxSize=<100 100>
     }}
   ) {isSetParameters=true}
   /NiagaraFluids/Modules/Grid2D/Grid2D_SetResolution(
    Grid:NiagaraDataInterfaceGrid2DCollection=$Emitter.VelocityGrid
    NumCellsMaxAxis=256
    Resolution Method:ENiagaraGrid2DResolution=::NewEnumerator1
    WorldGridExtents:Vector2f=<256 256>
   )
   &Grid2D_SetResolution(
    Grid:NiagaraDataInterfaceGrid2DCollection=$Emitter.PressureGrid
    OtherGrid:NiagaraDataInterfaceGrid2DCollection=$Emitter.VelocityGrid
    Resolution Method:ENiagaraGrid2DResolution=::NewEnumerator3
    WorldGridExtents:Vector2f=$Emitter.Grid2D_SetResolution.WorldGridExtents
   ) {module=Grid2D_SetResolution002}
   /SetVariables(Emitter.DeltaTime=0.033) {isSetParameters=true}
   /SetVariables(Emitter.SimRT=DataInterface() {propertyValues {
      MipMapGeneration=Disabled
      MipMapGenerationType=Linear
      OverrideRenderTargetFilter=TF_Default
      OverrideRenderTargetFormat=RTF_RGBA16f
      bInheritUserParameterSettings=false
      bOverrideFormat=true
      bPreviewRenderTarget=false
      RenderTargetUserParameter {parameter {name=None, typeDefHandle {registeredTypeIndex=4}, varData []}}
      Size=<0 0>
     }}) {isSetParameters=true}
   /NiagaraFluids/Modules/Grid2D/Grid2D_InitRT(Grid:NiagaraDataInterfaceGrid2DCollection=$Emitter.VelocityGrid, SimRT:NiagaraDataInterfaceRenderTarget2D=$Emitter.SimRT) {module=InitRTGrid2D}
   &self:FN_PIC_FLIP_Fluid_Stream002_2.NiagaraScratchPadContainer_0.ScratchModule_0() {module=ScratchModule}
   &self:FN_PIC_FLIP_Fluid_Stream002_2.NiagaraScratchPadContainer_0.ScratchModule_03_5(NumPressureIterations=8) {module=ScratchModule_03}
   /SetVariables() {isSetParameters=true}
   /SetVariables(Emitter.ThicknessRT=DataInterface() {propertyValues {
      MipMapGeneration=Disabled
      MipMapGenerationType=Linear
      OverrideRenderTargetFilter=TF_Default
      OverrideRenderTargetFormat=RTF_R16f
      bInheritUserParameterSettings=false
      bOverrideFormat=true
      bPreviewRenderTarget=false
      RenderTargetUserParameter {parameter {name=None, typeDefHandle {registeredTypeIndex=4}, varData []}}
      Size=<0 0>
     }}) {isSetParameters=true}
   &Grid2D_InitRT(Grid:NiagaraDataInterfaceGrid2DCollection=$Emitter.VelocityGrid, SimRT:NiagaraDataInterfaceRenderTarget2D=$Emitter.ThicknessRT)
   /SetVariables(
    Emitter.RasterizationGrid=DataInterface() {propertyValues {
      CellSize=1
      ClearBeforeNonIterationStage=true
      NumAttributes=1
      NumCellsMaxAxis=10
      Precision=1000
      ResetValue=100000000
      SetResolutionMethod=Independent
      NumCells=<3 3 3>
      WorldBBoxSize=<100 100 100>
     }}
    Emitter.RasterizationGridBack=DataInterface() {propertyValues {
      CellSize=1
      ClearBeforeNonIterationStage=true
      NumAttributes=1
      NumCellsMaxAxis=10
      Precision=1000
      ResetValue=0
      SetResolutionMethod=Independent
      NumCells=<3 3 3>
      WorldBBoxSize=<100 100 100>
     }}
   ) {isSetParameters=true}
   /NiagaraFluids/Modules/Grid3D/RasterizationGrid3D_SetResolution(
    Grid:NiagaraDataInterfaceRasterizationGrid3D=$Emitter.RasterizationGrid
    NumCellsX:NiagaraInt32=/Game/Effects/FlipSimulation/Niagara/DynamicInputs/GetGrid2DSize(Grid2D Collection:NiagaraDataInterfaceGrid2DCollection=$Emitter.VelocityGrid, Y:NiagaraBool=false)
    NumCellsY:NiagaraInt32=&GetGrid2DSize(Grid2D Collection:NiagaraDataInterfaceGrid2DCollection=$Emitter.VelocityGrid, Y:NiagaraBool=true)
    NumCellsZ=1
    WorldGridExtents:Vector3f=<1 1 1>
   )
   &RasterizationGrid3D_SetResolution(
    Grid:NiagaraDataInterfaceRasterizationGrid3D=$Emitter.RasterizationGridBack
    NumCellsX:NiagaraInt32=&GetGrid2DSize(Grid2D Collection:NiagaraDataInterfaceGrid2DCollection=$Emitter.VelocityGrid, Y:NiagaraBool=false)
    NumCellsY:NiagaraInt32=&GetGrid2DSize(Grid2D Collection:NiagaraDataInterfaceGrid2DCollection=$Emitter.VelocityGrid, Y:NiagaraBool=true)
    NumCellsZ=1
    WorldGridExtents:Vector3f=<1 1 1>
   ) {module=RasterizationGrid3D_SetResolution001}
  ]
  emitterUpdateScript [
   /NiagaraFN/Modules/Scalability/SignificantInstancesCull(Cull Response:ENiagaraExecutionStateManagement=::KillImmediately)
   &EmitterState()
   /NiagaraFluids/Modules/Grid2D/Grid2D_UpdateSimAttributes(Grid:NiagaraDataInterfaceGrid2DCollection=$Emitter.VelocityGrid, WorldGridSize:Vector2f=$Emitter.Grid2D_SetResolution.WorldGridExtents) {module=UpdateSimAttributes}
   /NiagaraFluids/Modules/Grid2D/Grid2D_CreateUnitToWorldTransform(
    Camera:NiagaraDataInterfaceCamera=$Emitter.CameraQuery
    Camera Facing=true
    Local Pivot:Vector2f=<0.5 0.85>
    Offset:NiagaraPosition=$Engine.Owner.Position
    Rotation:Quat4f=&self:FN_PIC_FLIP_Fluid_Stream002_2.NiagaraScratchPadContainer_0.ScratchDynamicInput_01_1(Rotation:Quat4f=$Engine.Owner.Rotation)
    WorldGridExtents:Vector2f=$Emitter.Grid2D_SetResolution.WorldGridExtents
   )
   /SetVariables(Emitter.Position:NiagaraPosition=&ConvertVectorToPosition(
     Input Position=$Emitter.Grid2D_CreateUnitToWorldTransform.LocalToWorldTranslation
     Conversion Operation:ENiagara_LWCConvertVecToPos=::NewEnumerator0
    )) {isSetParameters=true}
   /NiagaraFluids/Modules/Grid2D/Grid2D_SetBounds(
    EmitterProperties=DataInterface() {propertyValues {EmitterBinding {bindingMode=Self, emitterName=None}}}
    LocalPivot:Vector2f=$Emitter.Grid2D_CreateUnitToWorldTransform.LocalPivot
    WorldGridExtents:Vector2f=$Emitter.Grid2D_SetResolution.WorldGridExtents
   )
   /SetVariables(Emitter.OrthogonalVelocityMult=0.6) {isSetParameters=true}
   /NiagaraFluids/Modules/Grid2D/SplitMatrixIntoRows(Matrix:NiagaraMatrix=$Emitter.Grid2D_CreateUnitToWorldTransform.WorldToUnit)
   &SpawnRate(SpawnRate=50000, Use Spawn Probability=false)
   /Game/Effects/FlipSimulation/Niagara/Modules/Flip_SurfacingParameters(
    BaseColor_HighThickness:LinearColor=<rgba 0.0295 0.5777 1 1>
    BaseColor_LowThickness:LinearColor=<rgba 0 0.3333 1 1>
    CustomBaseColor_HighThickness=true
    CustomBaseColor_lowThickness=true
    CustomDarkFaceBrightness=false
    CustomEdgeOffset=true
    CustomEnvReflection_Tint=false
    CustomFakeSSS_Color=false
    CustomHDRIReflection_Tint=false
    CustomHueShiftDarkFace=false
    CustomHueShiftLightFace=false
    CustomLightFaceBrightness=false
    CustomRefractionStrength=false
    CustomScaleBlurDistance=true
    CustomShadowStrength=false
    CustomThicknessDivideNumber=true
    CustomThicknessEmissive_HighThickness=true
    CustomThicknessEmissive_LowThickness=true
    CustomTotalOpacityScale=true
    DebugBound=true
    Preset:ENiagara_FlipRenderPresets=::NewEnumerator1
    ScaleBlurDistance=0.8
    ThicknessDivide=60
    ThicknessEmissive_HighThickness:LinearColor=<rgba 0.0236 0.4562 0.8 1>
    ThicknessEmissive_LowThickness:LinearColor=<rgba 0 0.225 0.45 1>
    TotalOpacityScale:NiagaraFloat=&FloatFromCurve(
     FloatCurve=DataInterface() {propertyValues {Curve {keys [{interpMode=RCIM_Cubic, tangentMode=RCTM_Break, time=0, value=0}, <at 0.5 1>]}}}
     CurveIndex=$System.Age
    )
   )
   &SampleSkeletalMesh(
    AdditionalYawPitchRoll=false
    Bone Sampling Mode:ENiagaraBoneSamplingMode=::NewEnumerator4
    Skeletal Mesh:NiagaraDataInterfaceSkeletalMesh=$User.TargetPawn
   )
  ]
  particleSpawnScript [
   &InitializeParticle(
    Color=<rgba 0 0.3383 1 1>
    Lifetime=&RandomRangeFloat(Minimum=0.5, Maximum=0.7)
    Module.Write Color=false
    Module.Write Lifetime=true
    Module.Write Mass=false
    Module.Write Material Random=true
    Module.Write Position=false
    Module.Write Scale=false
    Module.Write SpriteRotation=false
    Module.Write SpriteSize=true
    Position=&SimulationPosition()
    Sprite Size=&Vector2DFromFloat(Value=2)
   )
   /Game/Effects/FlipSimulation/Niagara/Modules/Flip_Source(
    DeltaTime:NiagaraFloat=$Emitter.DeltaTime
    EnableSimpleSourceMoving=true
    InheritVelocityScale=0
    Radius:NiagaraFloat=&FloatFromCurve(
     FloatCurve=DataInterface() {propertyValues {Curve {keys [{interpMode=RCIM_Cubic, tangentMode=RCTM_Break, time=0, value=0.3}, <at 0.6 1>]}}}
     CurveIndex=$System.Age
     Scale Curve=15
    )
    SourceAge=0.05
    SourceCount=10
    SourceMovingRemap=DataInterface() {propertyValues {Curve {keys [<at 0 0>,<at 1 1>]}}}
   )
   /Game/Effects/FlipSimulation/Niagara/Modules/Flip_SourceLinearMovement(InheritVelocityScale=5, LinearVelocityWS:Vector3f=<0 0 -5>)
   &CurlNoiseForce(
    Noise Frequency=4
    Noise Strength=&FloatFromCurve(
     FloatCurve=DataInterface() {propertyValues {Curve {keys [{interpMode=RCIM_Cubic, tangentMode=RCTM_Break, time=0, value=0}, <at 0.9 1>]}}}
     CurveIndex=&Multiply_Float(A=$System.Age, B=1)
     Scale Curve=200
    )
    Pan Noise=true
    Pan Noise Field=<0 0 5>
   )
   &ShapeLocation(
    MaskSpawnGroup=false
    Override Seed=false
    Sphere Radius=&Multiply_Float(
     A=5
     B=&RemapRange(Input Value To Remap=$Particles.FlipSourceRandom, Input Min=0, Input Max=1, Output Min=0.5, Output Max=1)
    )
    UseOwnerScale=true
   )
   &AddVelocity(Velocity=<50 0 0>)
   &ApplyInitialForces()
   /SetVariables(Particles.Velocity=&Multiply_Vector(
     A=$Particles.Velocity
     B=&VectorFromFloat(Value=&ClampFloat(Float=&Power(a=&Divide_Float(A=0.0167, B=$Engine.DeltaTime), b=0.5), Min=0.5, Max=2))
    )) {isSetParameters=true}
   /SetVariables(Particles.LastEmitterPosition:NiagaraPosition=$Emitter.SampleSkeletalMesh.SampledPosition) {isSetParameters=true}
  ]
  particleUpdateScript [
   &ParticleState()
   &GravityForce(Gravity=<0 0 -700>)
   &PointForce(
    Force Origin=&AddVectorToPosition(Position=$Engine.Emitter.SimulationPosition, Vector=<0 0 0>)
    Force Strength=800
    Module.Use Falloff=true
    Module.UseLinearForce=false
   )
   &PointForce(
    Force Falloff Distance=70
    Force Origin=&AddVectorToPosition(Position=$Engine.Emitter.SimulationPosition, Vector=<0 0 -70>)
    Force Strength=-1500
    Module.Use Falloff=true
    Module.UseLinearForce=false
   ) {enabled=false, module=PointForce001}
   &Drag(Drag=3, Use Linear Drag=true, Use Rotational Drag=true)
   &CurlNoiseForce(
    Noise Strength=&FloatFromCurve(
     FloatCurve=DataInterface() {propertyValues {Curve {keys [{tangentMode=RCTM_User, time=0, value=1}, {tangentMode=RCTM_User, time=0.5, value=0}]}}}
     Scale Curve=500
    )
    Pan Noise=true
    Pan Noise Field=<1 1 1>
   ) {module=CurlNoiseForce001}
   /SetVariables(Transient.PhysicsForce:Vector3f=&Multiply_VectorByFloat(
     Vector=$Transient.PhysicsForce
     Float=&ClampFloat(
      Float=&Power(a=&Divide_Float(A=$Engine.DeltaTime, B=0.0167), b=0.5)
      Min=0.5
      Max=1.3
     )
    )) {isSetParameters=true}
   /SetVariables(Transient.StartPosition:NiagaraPosition=$Particles.Position) {isSetParameters=true}
   &SolveForcesAndVelocity(Clamp Velocity=false, DeltaTime:NiagaraFloat=$Emitter.DeltaTime, Limit Acceleration=false, Write to Presolve Properties=false)
   /SetVariables(Particles.Position:Vector3f=$Transient.StartPosition) {isSetParameters=true}
  ]
  rendererClasses [&NiagaraSpriteRendererProperties]
  renderers [
   {rendererClass=&NiagaraSpriteRendererProperties}
   {rendererClass=&NiagaraSpriteRendererProperties, rendererIndex=1}
   {rendererClass=&NiagaraSpriteRendererProperties, rendererIndex=2}
   {rendererClass=&NiagaraSpriteRendererProperties, rendererIndex=3}
   {rendererClass=&NiagaraSpriteRendererProperties, rendererIndex=4}
  ]
 }
}
summary {
 systemName=NS_Paprika_Consumable_BigShield_V2
 emitters [
  {emitterName=Drips_CPU_Low, rendererClasses [&NiagaraSpriteRendererProperties]}
  {emitterName=Drips_GPU, simTarget=GPUComputeSim, rendererClasses [&NiagaraSpriteRendererProperties]}
  {emitterName=FN_PIC_FLIP_Fluid_Stream002, simTarget=GPUComputeSim, rendererClasses [&NiagaraSpriteRendererProperties]}
  {bEnabled=false, emitterName=BakeFluid_PoringStart, rendererClasses [&NiagaraSpriteRendererProperties]}
  {emitterName=BakeFluid_NoneEpic, rendererClasses [&NiagaraSpriteRendererProperties]}
  {emitterName=BakeFluid_PoringStar_NoneEpic, rendererClasses [&NiagaraSpriteRendererProperties]}
 ]
 userVariables [User.TargetPawn {
   type=NiagaraDataInterfaceSkeletalMesh
   defaultValue {
    struct=NiagaraExt_VariableValue_DataInterface
    value {dataInterface=&self:NiagaraDataInterfaceSkeletalMesh_0}
   }
  }]
}
