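# Niagara effect for the Paprika medkit consumable, V2, producing short-lived liquid spurts and falling droplets.
# Two BakeFluid sprite emitters animate dynamic material parameters for the fluid body, while two GPU drip emitters add speed-stretched droplets, one steady and one tapering with system age.
# System.VelocityMultiplier scales drip velocity and second-fluid size and opacity; those three emitters also partly inherit source movement.
# The summary also lists a PIC/FLIP stream emitter and a TargetPawn skeletal-mesh parameter.
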NiagaraSystem /PaprikaConsumables/Assets/Effect/NS_Paprika_Consumable_Medkit_V2
dependencies=null
emitters {
 BakeFluid_NoneEpic {
  emitterUpdateScript [
   &EmitterState(Life Cycle Mode=::NewEnumerator1, Loop Behavior=::NewEnumerator1, Loop Delay=0.25, Loop Duration=1, UseLoopDelay=true)
   &SpawnRate(
    SpawnRate=&FloatFromCurve(
     FloatCurve=DataInterface() {propertyValues {Curve {keys [{arriveTangent=3.4706, interpMode=RCIM_Cubic, leaveTangent=3.4706, tangentMode=RCTM_User, time=0, value=0.5}, <at 0.2 1 RCIM_Cubic>]}}}
     CurveIndex=$System.Age
     Scale Curve=15
    )
    Use Spawn Probability=false
   )
  ]
  particleSpawnScript [
   &InitializeParticle(
    Color=<rgba 1 1 1 0.6>
    Lifetime Max=0.3
    Lifetime Min=0.3
    Lifetime Mode=::NewEnumerator1
    Sprite Size Mode=::NewEnumerator1
    Sprite UV Mode=::NewEnumerator3
    Uniform Sprite Size Max=55
    Uniform Sprite Size Min=35
    UsePositionOffset=false
    Write Material Random=true
   )
   &AddVelocity(Velocity=<30 0 50>) {enabled=false}
   &CameraOffset(Camera Offset Amount=10)
  ]
  particleUpdateScript [
   &ParticleState()
   &ScaleSpriteSize(Scale Factor=<0.6 2>, Scale Sprite Size Mode=::NewEnumerator2) {module=ScaleSpriteSize001}
   &ScaleSpriteSize(
    Uniform Curve Scale=0.65
    Uniform Curve Sprite Scale=DataInterface() {propertyValues {Curve {keys [
        {arriveTangent=0.0076, interpMode=RCIM_Cubic, leaveTangent=0.0076, tangentMode=RCTM_User, time=0, value=0.3}
        {interpMode=RCIM_Cubic, tangentMode=RCTM_User, time=1, value=0.6}
       ]}}}
   )
   &DynamicMaterialParameters(
    Index 0 Param 1=&FloatFromCurve(FloatCurve=DataInterface() {propertyValues {Curve {keys [
         {arriveTangent=-0.2479, interpMode=RCIM_Cubic, leaveTangent=-0.2479, tangentMode=RCTM_User, time=0, value=0}
         {arriveTangent=0.6296, interpMode=RCIM_Cubic, leaveTangent=0.6296, tangentMode=RCTM_User, time=1, value=0.5}
        ]}}})
    Index 0 Param 2=&FloatFromCurve(FloatCurve=DataInterface() {propertyValues {Curve {keys [{arriveTangent=0.4361, time=0.2, value=0.05}, <at 1 0.7>]}}}, Scale Curve=2)
    Index 0 Param 3=&FloatFromCurve(FloatCurve=DataInterface() {propertyValues {Curve {keys [
         {arriveTangent=0.4361, interpMode=RCIM_Cubic, tangentMode=RCTM_Break, time=0.5, value=0}
         {arriveTangent=-1.6488, interpMode=RCIM_Cubic, tangentMode=RCTM_Break, time=1, value=-0.5}
        ]}}})
    Module.Param0WriteEnabled=true
    Module.Param1WriteEnabled=true
    Module.Param2WriteEnabled=true
    Module.Param3WriteEnabled=true
   )
   &GravityForce(Gravity=<0 0 -200>)
   &SolveForcesAndVelocity(Clamp Velocity=false, Limit Acceleration=false)
   &ScaleSpriteSizeBySpeed(Max Scale Factor=<3 10>)
   &ScaleColor(
    Scale Alpha=&FloatFromCurve(FloatCurve=DataInterface() {propertyValues {Curve {keys [{interpMode=RCIM_Cubic, tangentMode=RCTM_User, time=0, value=1}, {interpMode=RCIM_Cubic, tangentMode=RCTM_User, time=1, value=0}]}}})
    ScaleA=true
    ScaleRGB=true
   )
   &ScaleColor(
    Scale Alpha=&FloatFromCurve(
     FloatCurve=DataInterface() {propertyValues {Curve {keys [
         {tangentMode=RCTM_User, time=0, value=1}
         {arriveTangent=-0.0405, interpMode=RCIM_Cubic, leaveTangent=-0.0405, tangentMode=RCTM_User, time=0.7009, value=1}
         {arriveTangent=-8.0089, interpMode=RCIM_Cubic, leaveTangent=-8.0089, tangentMode=RCTM_User, time=1, value=0}
        ]}}}
     CurveIndex=$Engine.System.Age
    )
    ScaleA=true
    ScaleRGB=true
   ) {module=ScaleColor001}
  ]
  rendererClasses [&NiagaraSpriteRendererProperties]
  renderers [{rendererClass=&NiagaraSpriteRendererProperties}]
 }
 BakeFluid_NoneEpic001 {
  emitterUpdateScript [
   &EmitterState(Life Cycle Mode=::NewEnumerator1, Loop Behavior=::NewEnumerator1, Loop Duration=0.8, UseLoopDelay=false)
   &SpawnRate(
    SpawnRate=&FloatFromCurve(
     FloatCurve=DataInterface() {propertyValues {Curve {keys [{arriveTangent=3.4706, interpMode=RCIM_Cubic, leaveTangent=3.4706, tangentMode=RCTM_User, time=0, value=0}, <at 0.5 1 RCIM_Cubic>]}}}
     CurveIndex=$System.Age
     Scale Curve=8
    )
    Use Spawn Probability=false
   )
  ]
  particleSpawnScript [
   &InitializeParticle(
    Color=<rgba 1 1 1 0.6>
    Lifetime Max=0.6
    Lifetime Min=0.4
    Lifetime Mode=::NewEnumerator1
    Sprite Rotation Mode=::NewEnumerator2
    Sprite Size Mode=::NewEnumerator1
    Sprite UV Mode=::NewEnumerator3
    Uniform Sprite Size Max=55
    Uniform Sprite Size Min=40
    UsePositionOffset=false
    Write Material Random=true
   )
   &AddVelocity(Module.Use Falloff Distance=false, Override Seed=false, Velocity Mode=::NewEnumerator1, Velocity Speed=10)
   &CameraOffset(Camera Offset Amount=20)
   &ShapeLocation(
    MaskSpawnGroup=false
    Offset=<0 0 -8>
    Offset Coordinate Space=::World
    Offset Mode=::NewEnumerator0
    Override Seed=false
    Sphere Radius=1
    UseOwnerScale=true
   )
  ]
  particleUpdateScript [
   &ParticleState()
   &ScaleSpriteSize(Scale Factor=<2 2>, Scale Sprite Size Mode=::NewEnumerator2) {module=ScaleSpriteSize001}
   &ScaleSpriteSize(
    Uniform Curve Scale=&Multiply_Float(A=2.5, B=$System.VelocityMultiplier)
    Uniform Curve Sprite Scale=DataInterface() {propertyValues {Curve {keys [
        {arriveTangent=0.1552, interpMode=RCIM_Cubic, leaveTangent=0.1552, tangentMode=RCTM_User, time=0, value=0.1}
        {arriveTangent=1.3084, interpMode=RCIM_Cubic, leaveTangent=1.3084, tangentMode=RCTM_User, time=1, value=0.6}
       ]}}}
   )
   &DynamicMaterialParameters(
    Index 0 Param 1=&FloatFromCurve(FloatCurve=DataInterface() {propertyValues {Curve {keys [{tangentMode=RCTM_User, time=0, value=0.3}, {tangentMode=RCTM_User, time=1, value=1}]}}})
    Index 0 Param 2=&FloatFromCurve(
     FloatCurve=DataInterface() {propertyValues {Curve {keys [
         {arriveTangent=1.6435, interpMode=RCIM_Cubic, leaveTangent=1.6435, tangentMode=RCTM_User, time=0.2, value=0.05}
         {arriveTangent=0.8125, interpMode=RCIM_Cubic, leaveTangent=0.8125, time=0.6075, value=0.4127}
         <at 1 0.7 RCIM_Cubic>
        ]}}}
     Scale Curve=2
    )
    Index 0 Param 3=&FloatFromCurve(FloatCurve=DataInterface() {propertyValues {Curve {keys [
         {arriveTangent=0.4361, interpMode=RCIM_Cubic, tangentMode=RCTM_Break, time=0.5, value=0}
         {arriveTangent=-1.6488, interpMode=RCIM_Cubic, tangentMode=RCTM_Break, time=1, value=-0.5}
        ]}}})
    Module.Param0WriteEnabled=true
    Module.Param1WriteEnabled=true
    Module.Param2WriteEnabled=true
    Module.Param3WriteEnabled=true
   )
   &InheritSourceMovement(Applied Position Scale=<0.7 0.7 0.7>, ApplyForce=false, ApplyPosition=true, ApplyPositionFalloff=false)
   &GravityForce(Gravity=<0 0 -900>)
   &Drag(Drag=3, Use Linear Drag=true, Use Rotational Drag=true)
   &SolveForcesAndVelocity(Clamp Velocity=false, Limit Acceleration=false)
   &ScaleSpriteSizeBySpeed(Max Scale Factor=<3 10>) {enabled=false}
   &ScaleColor(
    Scale Alpha=&FloatFromCurve(
     FloatCurve=DataInterface() {propertyValues {Curve {keys [
         {interpMode=RCIM_Cubic, tangentMode=RCTM_User, time=0, value=0}
         <at 0.2 1>
         {arriveTangent=-1.8683, interpMode=RCIM_Cubic, leaveTangent=-1.8683, tangentMode=RCTM_User, time=0.3, value=0.45}
         {interpMode=RCIM_Cubic, tangentMode=RCTM_User, time=1, value=0}
        ]}}}
     Scale Curve=&Multiply_Float(A=$System.VelocityMultiplier, B=0.6)
    )
    ScaleA=true
    ScaleRGB=true
   )
  ]
  rendererClasses [&NiagaraSpriteRendererProperties]
  renderers [{rendererClass=&NiagaraSpriteRendererProperties}]
 }
 Drips_GPU {
  simTarget=GPUComputeSim
  emitterUpdateScript [
   &EmitterState()
   &SpawnRate(SpawnRate=40, Use Spawn Probability=false)
  ]
  particleSpawnScript [
   /Niagara/Modules/Spawn/Initialization/InitializeParticle(
    Lifetime=&RandomRangeFloat(Minimum=0.5, Maximum=0.8, Evaluation Type=::NewEnumerator1)
    Module.Write Color=true
    Module.Write Lifetime=true
    Module.Write Mass=true
    Module.Write Material Random=true
    Module.Write Position=false
    Module.Write Scale=false
    Module.Write SpriteRotation=false
    Module.Write SpriteSize=true
    Sprite Size=&RandomRangeVector2D(Minimum=<0.5 0.8>, Maximum=<1.5 2>, Evaluation Type=::NewEnumerator1)
   ) {enabled=false}
   /Niagara/Modules/Spawn/Initialization/V2/InitializeParticle(
    Lifetime Max=0.5
    Lifetime Min=0.2
    Lifetime Mode=::NewEnumerator1
    Sprite Size=&RandomRangeVector2D(Minimum=<0.5 0.8>, Maximum=<1.5 2>)
    Sprite Size Mode=::NewEnumerator4
    UsePositionOffset=false
    Write Material Random=true
   ) {module=InitializeParticle001}
   &AddVelocity(
    Coordinate Space=::World
    Scale Added Velocity=&VectorFromFloat(Value=$System.VelocityMultiplier)
    Velocity=&RandomRangeVector(Minimum=<-100 -100 -10>, Maximum=<100 100 -100>, Evaluation Type=::NewEnumerator1)
   )
   /Niagara/Modules/Spawn/Location/SphereLocation(MaskSpawnGroup=false, Module.Override Randomness=false, Sphere Radius=1, Surface Only=false) {enabled=false}
   &ShapeLocation(
    MaskSpawnGroup=false
    Offset=<0 0 -5>
    Offset Coordinate Space=::World
    Offset Mode=::NewEnumerator0
    Override Seed=false
    Sphere Radius=1
    UseOwnerScale=true
   )
   &AddVelocity(
    Module.Use Falloff Distance=false
    Override Seed=false
    Velocity Mode=::NewEnumerator1
    Velocity Speed=&Multiply_Float(A=&RandomRangeFloat(Minimum=0.5, Maximum=1), B=250)
   ) {enabled=false, module=AddVelocity001}
   &CameraOffset(Camera Offset Amount=10)
  ]
  particleUpdateScript [
   &ParticleState()
   &InheritSourceMovement(Applied Position Scale=<0.8 0.8 0.8>, ApplyForce=false, ApplyPosition=true, ApplyPositionFalloff=false)
   &Drag(Use Linear Drag=true, Use Rotational Drag=false)
   &AccelerationForce(Acceleration=<0 0 -980>)
   &Collision(
    Age Colliding Particles=false
    Friction=0
    Kill Offscreen Particles=true
    KillOnCollision=false
    Module.Control Rest State Via Surface Normal=false
    Module.EnableMaxCollisionCount=false
    Module.Randomize Collision Normal Vector=false
    Restitution=0.1
   ) {enabled=false, module=Collision001}
   &SolveForcesAndVelocity(Clamp Velocity=false, Limit Acceleration=false)
   &Color(Color=&MakeLinearColorFromVectorAndFloat(
     "Vector (RGB)"=&VectorFromCurve(VectorCurve=DataInterface() {propertyValues {XCurve {keys [<at 0 1>,<at 1 1>]}, YCurve {keys [<at 0 1>,<at 1 1>]}, ZCurve {keys [<at 0 1>,<at 1 1>]}}})
     "Float (Alpha)"=&ClampFloat(Float=&FloatFromCurve(FloatCurve=DataInterface() {propertyValues {Curve {keys [{arriveTangent=-1, time=0, value=1}, <at 1 0>]}}}))
    ))
   &ScaleSpriteSizeBySpeed(Max Scale Factor=<1 10>, Source Velocity:Vector3f=&Multiply_Vector(A=<0 0.05 1>, B=$Particles.Velocity))
   &ScaleSpriteSize(Scale Factor=&Vector2DFromCurve(Vector2Curve=DataInterface() {propertyValues {XCurve {keys [<at 0.5 1>,<at 1 0>]}, YCurve {keys [<at 0.5 1>,<at 1 0>]}}}, ScaleCurve=<2 2>))
  ]
  rendererClasses [&NiagaraSpriteRendererProperties]
  renderers [{rendererClass=&NiagaraSpriteRendererProperties}]
 }
 Drips_GPU001 {
  simTarget=GPUComputeSim
  emitterUpdateScript [
   &EmitterState(Life Cycle Mode=::NewEnumerator1, Loop Behavior=::NewEnumerator1, Loop Duration=1, UseLoopDelay=true)
   &SpawnRate(
    SpawnRate=&FloatFromCurve(
     FloatCurve=DataInterface() {propertyValues {Curve {keys [
         {tangentMode=RCTM_User, time=0, value=1}
         {arriveTangent=-0.0405, interpMode=RCIM_Cubic, leaveTangent=-0.0405, tangentMode=RCTM_User, time=0.7009, value=1}
         {arriveTangent=-8.0089, interpMode=RCIM_Cubic, leaveTangent=-8.0089, tangentMode=RCTM_User, time=1, value=0}
        ]}}}
     CurveIndex=$Engine.System.Age
     Scale Curve=100
    )
    Use Spawn Probability=false
   )
  ]
  particleSpawnScript [
   /Niagara/Modules/Spawn/Initialization/InitializeParticle(
    Lifetime=&RandomRangeFloat(Minimum=0.5, Maximum=0.8, Evaluation Type=::NewEnumerator1)
    Module.Write Color=true
    Module.Write Lifetime=true
    Module.Write Mass=true
    Module.Write Material Random=true
    Module.Write Position=false
    Module.Write Scale=false
    Module.Write SpriteRotation=false
    Module.Write SpriteSize=true
    Sprite Size=&RandomRangeVector2D(Minimum=<0.5 0.8>, Maximum=<1.5 2>, Evaluation Type=::NewEnumerator1)
   ) {enabled=false}
   /Niagara/Modules/Spawn/Initialization/V2/InitializeParticle(
    Lifetime Max=0.5
    Lifetime Min=0.2
    Lifetime Mode=::NewEnumerator1
    Sprite Size=&RandomRangeVector2D(Minimum=<0.5 0.8>, Maximum=<1.5 2>)
    Sprite Size Mode=::NewEnumerator4
    UsePositionOffset=false
    Write Material Random=true
   ) {module=InitializeParticle001}
   &AddVelocity(
    Coordinate Space=::World
    Scale Added Velocity=&VectorFromFloat(Value=$System.VelocityMultiplier)
    Velocity=&RandomRangeVector(Minimum=<-100 -100 -10>, Maximum=<100 100 -100>, Evaluation Type=::NewEnumerator1)
   )
   &SphereLocation(MaskSpawnGroup=false, Module.Override Randomness=false, Sphere Radius=1, Surface Only=false) {enabled=false}
   &ShapeLocation(
    MaskSpawnGroup=false
    Offset=<0 0 -5>
    Offset Coordinate Space=::World
    Offset Mode=::NewEnumerator0
    Override Seed=false
    Sphere Radius=1
    UseOwnerScale=true
   )
   &AddVelocity(
    Module.Use Falloff Distance=false
    Override Seed=false
    Velocity Mode=::NewEnumerator1
    Velocity Speed=&Multiply_Float(A=&RandomRangeFloat(Minimum=0.5, Maximum=1), B=250)
   ) {enabled=false, module=AddVelocity001}
   &CameraOffset(Camera Offset Amount=10)
  ]
  particleUpdateScript [
   &ParticleState()
   &InheritSourceMovement(Applied Position Scale=<0.7 0.7 0.7>, ApplyForce=false, ApplyPosition=true, ApplyPositionFalloff=false)
   &Drag(Use Linear Drag=true, Use Rotational Drag=false)
   &AccelerationForce(Acceleration=<0 0 -980>)
   &Collision(
    Age Colliding Particles=false
    Friction=0
    Kill Offscreen Particles=true
    KillOnCollision=false
    Module.Control Rest State Via Surface Normal=false
    Module.EnableMaxCollisionCount=false
    Module.Randomize Collision Normal Vector=false
    Restitution=0.1
   ) {enabled=false, module=Collision001}
   &SolveForcesAndVelocity(Clamp Velocity=false, Limit Acceleration=false)
   &Color(Color=&MakeLinearColorFromVectorAndFloat(
     "Vector (RGB)"=&VectorFromCurve(VectorCurve=DataInterface() {propertyValues {XCurve {keys [<at 0 1>,<at 1 1>]}, YCurve {keys [<at 0 1>,<at 1 1>]}, ZCurve {keys [<at 0 1>,<at 1 1>]}}})
     "Float (Alpha)"=&ClampFloat(Float=&FloatFromCurve(FloatCurve=DataInterface() {propertyValues {Curve {keys [{arriveTangent=-1, time=0, value=1}, <at 1 0>]}}}))
    ))
   &ScaleSpriteSizeBySpeed(Max Scale Factor=<1 10>, Source Velocity:Vector3f=&Multiply_Vector(A=<0 0.05 1>, B=$Particles.Velocity))
   &ScaleSpriteSize(Scale Factor=&Vector2DFromCurve(
     Vector2Curve=DataInterface() {propertyValues {XCurve {keys [<at 0.5 1>,<at 1 0>]}, YCurve {keys [<at 0.5 1>,<at 1 0>]}}}
     ScaleCurve=<1.5 1.5>
    ))
  ]
  rendererClasses [&NiagaraSpriteRendererProperties]
  renderers [{rendererClass=&NiagaraSpriteRendererProperties}]
 }
}
summary {
 systemName=NS_Paprika_Consumable_Medkit_V2
 emitters [
  {emitterName=FN_PIC_FLIP_Fluid_Stream002, simTarget=GPUComputeSim, rendererClasses [&NiagaraSpriteRendererProperties]}
  {emitterName=BakeFluid_NoneEpic, rendererClasses [&NiagaraSpriteRendererProperties]}
  {emitterName=Drips_GPU, simTarget=GPUComputeSim, rendererClasses [&NiagaraSpriteRendererProperties]}
  {emitterName=BakeFluid_NoneEpic001, rendererClasses [&NiagaraSpriteRendererProperties]}
  {emitterName=Drips_GPU001, simTarget=GPUComputeSim, rendererClasses [&NiagaraSpriteRendererProperties]}
 ]
 userVariables [User.TargetPawn {
   type=NiagaraDataInterfaceSkeletalMesh
   defaultValue {
    struct=NiagaraExt_VariableValue_DataInterface
    value {dataInterface=&self:NiagaraDataInterfaceSkeletalMesh_0}
   }
  }]
}
