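# Sound cue for the landing of tossed meds in the PaprikaConsumables plugin, playing a single sound wave.
# A Modulator node at the root wraps one Wave Player that plays Meds_Tossed_01.
# The cue uses the Fort_Consumables sound class and the shared Athena consumable-use attenuation asset.
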SoundCue /PaprikaConsumables/Assets/Sound/Meds_Tossed_Land_Cue
root=Modulator_0
nodes [{id=Modulator_0, children [WavePlayer_0]}, {id=WavePlayer_0, properties {soundWaveAssetPtr=&/Meds_Tossed_01}}]
properties {
 attenuationSettings=&Fort_Consumable_Use_Athena_Att
 maxDistance=2250
 soundClassObject=&Fort_Consumables
 volumeMultiplier=1
}
