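# ShieldPot_Tossed_Cue is the sound cue for tossing a shield potion consumable.
# It mixes two layers at equal volume, a shield-pot toss foley and a generic consumable toss whoosh, then passes them through a Modulator at the root.
# It uses the Athena consumable-use attenuation and the consumables sound class, with a boosted overall volume.
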SoundCue /PaprikaConsumables/Assets/Sound/ShieldPot_Tossed_Cue
root=Modulator_1
nodes [
 {id=Mixer_0, children [WavePlayer_1,WavePlayer_2], properties {inputVolume [1,1]}}
 {id=Modulator_1, children [Mixer_0]}
 {id=WavePlayer_1, properties {soundWaveAssetPtr=&/ShieldPot_Tossed_Foley_01}}
 {id=WavePlayer_2, properties {soundWaveAssetPtr=&/Consumable_Tossed_Whoosh_01}}
]
properties {
 attenuationSettings=&Fort_Consumable_Use_Athena_Att
 maxDistance=2250
 soundClassObject=&Fort_Consumables
 volumeMultiplier=1.5
}
